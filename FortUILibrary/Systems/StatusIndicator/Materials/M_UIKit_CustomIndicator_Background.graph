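# Translucent UI material that draws the procedural, outlined backplate for a custom status indicator.
# The shape is a centered diamond from absolute scaled-to-fit UVs, intersected with a positionable, angled two-plane wedge mask to cut a notch.
# A sharpness-scaled distance field gives the antialiased mask, and subtracting outline thickness from it blends outline color into fill color.
# A vertex transform scales the quad by SpillOffScale, enlarging it past the widget bounds by default.
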Material /FortUILibrary/Systems/StatusIndicator/Materials/M_UIKit_CustomIndicator_Background
graph {
 properties {bAutomaticallySetUsageInEditor=false, blendMode=BLEND_Translucent, materialDomain=MD_UI, usedShadingModels=Default Lit}
 nodes [
  n1 = NamedRerouteUsage()
  n2 = ScalarParameter() {group=General, parameterName=Outline Thickness}
  LinearInterpolate(
   A=VectorParameter() {group=General, parameterName=OutlineColor, defaultValue=<rgba 0 0 0 1>}.RGB
   B=VectorParameter() {group=General, parameterName=Color, defaultValue=<rgba 1 1 1 1>}.RGB
   Alpha=Saturate(None=Subtract(A=@n1, B=@n2))
  )
  MaterialFunctionCall!(Anchor, Translation, Padding, Scale=ScalarParameter() {defaultValue=1.35, group=General, parameterName=SpillOffScale}) {materialFunction=/FortUILibrary/MaterialFunctions/Basics/MF_UI_VertexTransform}
  n3 = Add(A=MaterialFunctionCall() {materialFunction=&MF_UI_UVScaling}.Scaled to Fit) {constB=-0.5}
  n4 = Abs(None=@n3)
  n5 = ComponentMask(None=@n4) {r=true}
  n6 = ComponentMask(None=@n4) {g=true}
  n7 = Add(A=OneMinus(None=Add(A=@n5, B=@n6))) {constB=-0.5}
  n8 = ScalarParameter() {defaultValue=16, group=General, parameterName=Sharpness}
  n9 = Add(
   A=@n3
   B=AppendVector(
    A=ScalarParameter() {defaultValue=-0.1, group=Wedge, parameterName=WedgePositionX}
    B=ScalarParameter() {defaultValue=-0.32, group=Wedge, parameterName=WedgePositionY}
   )
  )
  n10 = ComponentMask(None=@n9) {r=true}
  n11 = ComponentMask(None=@n9) {g=true}
  n12 = ScalarParameter() {defaultValue=0.1, group=Wedge, parameterName=WedgeAngle}
  n13 = Max(A=Multiply(A=Add(A=@n10, B=@n11)) {constB=-1}, B=Add(A=@n10, B=Multiply(A=@n11, B=@n12) {constB=0.1}))
  NamedRerouteDeclaration(None=Min(A=Multiply(A=@n7, B=@n8) {constB=32}, B=Multiply(A=@n8, B=@n13))) {name=Final Backplate Shape, nodeColor=<rgba 1 0 0.2109 1>}
  Saturate(None=@n1)
 ]
}
