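# UI material that draws the rounded background of a status indicator, with an optional colored outline.
# An inner rounded mask, shrunk by a pixel-based Outline Thickness, separates Color from OutlineColor.
# A full-size mask at half opacity is combined with the inner mask through Max, making the outline ring semi-transparent.
# Floor and Saturate turn thickness into a branchless toggle, so values below one show only the fill.
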Material /FortUILibrary/Systems/StatusIndicator/Materials/M_UIKit_StatusIndicator_Background
graph {
 properties {
  bUsedWithStaticMesh=false
  blendMode=BLEND_Translucent
  materialDomain=MD_UI
  pixelDepthOffsetMode=PDOM_Legacy
  usedShadingModels=Default Lit
 }
 nodes [
  n1 = MaterialFunctionCall() {materialFunction=&GetUserInterfaceUV}
  n2 = NamedRerouteUsage()
  n3 = Saturate(None=MaterialFunctionCall!(Value PX=Multiply(A=@n2) {constB=2}, UV Pixel Size, X or Y) {materialFunction=/FortUILibrary/MaterialFunctions/Basics/MF_UI_PixelsToNormalized}."Normalized (>1px)")
  n4 = Constant() {r=0.5}
  n5 = Constant()
  n6 = MaterialFunctionCall(
   UV=@n1.Normalized UV
   Size=LinearInterpolate(B=OneMinus(None=@n3), Alpha=NamedRerouteUsage()) {constA=1}
   Corner Radius=@n4
   Radius in Pixels=@n5
  ) {materialFunction=&MF_UI_CoreUI_RoundedMask}
  n7 = NamedRerouteDeclaration(None=ScalarParameter() {defaultValue=2, parameterName=Outline Thickness}) {name=Outline Size, nodeColor=<rgba 0 0 0 1>}
  n8 = NamedRerouteDeclaration(None=Saturate(None=Floor(None=@n7))) {name=Is Displaying Outline, nodeColor=<rgba 0 0 0 1>}
  LinearInterpolate(
   A=VectorParameter() {parameterName=Color, defaultValue=<rgba 1 1 1 1>}.RGB
   B=VectorParameter() {parameterName=OutlineColor, defaultValue=<rgba 1 0 0.0263 1>}.RGB
   Alpha=LinearInterpolate(B=OneMinus(None=@n6.Shape Mask), Alpha=@n8) {desc=Controls Outline Display}
  )
  n9 = MaterialFunctionCall(UV=@n1.Normalized UV, Corner Radius=@n4, Radius in Pixels=@n5) {materialFunction=&MF_UI_CoreUI_RoundedMask}
  LinearInterpolate(A=@n6.Shape Mask, B=Max(A=@n6.Shape Mask, B=Multiply(A=@n9.Shape Mask) {constB=0.5, desc=Outline Opacity}), Alpha=@n8)
 ]
}
