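# UI material for a status indicator background, drawing a tinted rounded shape with a glow in alpha-composite mode.
# A rounded-mask SDF feeds an antialiased inner mask and an eased smoothstep falloff.
# Static Display switches between fixed padding with adjustable glow opacity and a looping, time-eased size with fading opacity.
# Blend Color and Alpha optionally premultiplies Color by that opacity, and vertex alpha scales the result.
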Material /FortUILibrary/Systems/StatusIndicator/Materials/M_UIKit_StatusIndicator_Bg
graph {
 properties {
  bUsedWithStaticMesh=false
  blendMode=BLEND_AlphaComposite
  materialDomain=MD_UI
  pixelDepthOffsetMode=PDOM_Legacy
  usedShadingModels=Default Lit
 }
 nodes [
  n1 = VectorParameter() {parameterName=Color, defaultValue=<rgba 0.9933 0 1 1>}
  n2 = Constant()
  n3 = ScalarParameter() {defaultValue=0.5, parameterName=Glow Opacity, sliderMax=1}
  n4 = Frac(None=Time())
  n5 = StaticSwitchParameter(
   True=OneMinus(None=ScalarParameter() {defaultValue=0.15, parameterName=Glow Padding, sliderMax=1})
   False=NamedRerouteDeclaration(None=MaterialFunctionCall("Value (0-1)"=@n4) {materialFunction=&MF_UI_EaseCurves}.EaseOutSine) {name=Time, nodeColor=<rgba 0 0 0 1>}
  ) {defaultValue=true, parameterName=Static Display}
  n6 = MaterialFunctionCall(
   UV=MaterialFunctionCall() {materialFunction=&GetUserInterfaceUV}.Normalized UV
   Size=@n5
   Corner Radius=Constant() {r=0.5}
   Radius in Pixels=Constant()
  ) {materialFunction=&MF_UI_CoreUI_RoundedMask}
  n7 = MaterialFunctionCall(DistanceField=@n6.SDF) {materialFunction=&MF_UI_AntiAliasing}
  n8 = NamedRerouteUsage()
  n9 = StaticSwitchParameter(
   True=LinearInterpolate(A=Multiply(A=@n2, B=@n3) {constB=0.5, desc=Glow Opacity}, Alpha=@n7.Inner)
   False=Multiply(A=Multiply(A=@n7.Inner) {constB=0.5}, B=OneMinus(None=@n8))
  ) {defaultValue=true, parameterName=Static Display}
  Multiply(A=VertexColor().A, B=StaticSwitchParameter(True=Multiply(A=@n1.RGB, B=@n9), False=@n1.RGB) {parameterName=Blend Color and Alpha})
  n10 = OneMinus(None=@n5)
  n11 = Multiply(A=@n10) {constB=-1}
  OneMinus(None=MaterialFunctionCall("Value (0-1)"=SmoothStep(Min=@n11, Max=@n10, Value=@n6.SDF)) {materialFunction=&MF_UI_EaseCurves}.EaseInOutSine)
 ]
}
