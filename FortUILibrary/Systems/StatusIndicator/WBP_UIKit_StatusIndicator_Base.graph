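# Base status indicator badge widget on FortBangWrapper_NUI, showing an icon, dot or count by configuration, with type-driven colors.
# Bang state changes drive count and visibility unless overridden; icon mode switches to text above one, capped as "N+" past CountCeiling.
# SetIndicatorConfiguration rebuilds the Image or CommonTextBlock at runtime over a dynamic-material background, sized by mode, digit count and mobile multipliers.
# TriggerIntro and TriggerReset play the opacity fade forward and reversed.
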WidgetBlueprint /FortUILibrary/Systems/StatusIndicator/WBP_UIKit_StatusIndicator_Base
animations [Anim_Intro {
  binding_count=1
  length_seconds=0.2
  object=&self:Anim_Intro
  sequence {
   bindings [Root {
     id=#1
     display_name=GridPanel_Root
     possessed_class=GridPanel
     tracks [{
       type=MovieSceneFloatTrack
       display_name=Render Opacity
       sections [{type=MovieSceneFloatSection, end_frame=12, start_frame=0, channels [None {type=MovieSceneScriptingFloatChannel, keys [[0,0],[11.999,1]]}]}]
      }]
    }]
   display_rate {numerator=60}
  }
 }]
graphs [
 CheckCurrentConfiguration {nodes [
   Inputs()
   > n1 = Switch on E_UI_StatusIndicator_Configuration!(
    execute
    Selection:E_UI_StatusIndicator_Configuration Enum=Get Configuration() -> (Configuration:E_UI_StatusIndicator_Configuration Enum=NewEnumerator0)
   ) -> (NewEnumerator0:Exec, NewEnumerator2:Exec, NewEnumerator3:Exec) {is=title K2Node_SwitchEnum}
   > n2 = Branch(Condition="integer > integer"(A=Get Count() -> (Count:int=0), B=1))
   Outputs!(Image:Exec=@n2.else, Dot:Exec=@n1.NewEnumerator2, Text:Exec=[@n2.then, @n1.NewEnumerator3])
  ]}
 EventGraph {nodes [
   n1 = OverrideCount!() -> (OutputDelegate, then, Count:int) {is=title K2Node_CustomEvent}
   > n2 = Set Count(Count:int=@n1.Count) -> (Output_Get:int=0)
   > n3 = Branch(Condition=Get OverrideStatusIndicatorVisibility() -> (OverrideStatusIndicatorVisibility:bool))
   n4 = SetVisibility(execute=@n3.else, InVisibility=Select(Option 0=Collapsed, Option 1=HitTestInvisible, Index="integer > integer"(A=@n2)))
   SetIndicatorConfiguration!(
    execute=[@n12.then, @n4.then]
    self
    InCount:int=Get Count() -> (Count=0)
    IndicatorConfig:E_UI_StatusIndicator_Configuration Enum=Get Configuration() -> (Configuration=NewEnumerator0)
   ) -> (then)
   n5 = OverrideCountCeiling!() -> (OutputDelegate, then, Ceiling:int) {is=title K2Node_CustomEvent}
   > Set CountCeiling(CountCeiling:int=@n5.Ceiling) -> (Output_Get:int=0)
   n6 = SetAsyncBackground!() -> (OutputDelegate, then, MaterialInterfaceValue:Material Interface Soft Object Reference) {is=title K2Node_CustomEvent}
   > Branch(Condition=IsValidSoftObjectReference(SoftObjectReference=@n6.MaterialInterfaceValue))
   > n7 = Async Load Asset(Asset=@n6.MaterialInterfaceValue)
   n8 = Cast To MaterialInterface!(execute=@n7.Completed, Object=@n7) -> (then, CastFailed, AsMaterial Interface:Material Interface Object Reference) {is=title K2Node_DynamicCast}
   > SetBackground!(execute, self, Material:Material Interface Object Reference=@n8) -> (then)
   > SetIndicatorType!(execute, self, IndicatorType:E_UI_StatusIndicator_Type Enum=Get Type() -> (Type:E_UI_StatusIndicator_Type Enum=NewEnumerator0)) -> (then)
   > SetIndicatorConfiguration(InCount=Get Count() -> (Count=0), IndicatorConfig=Get Configuration() -> (Configuration=NewEnumerator0))
   n9 = Event OnBangStateChanged!() -> (OutputDelegate, then, bEnabled:bool, Count:int=0) {is=title K2Node_Event}
   > n10 = Branch(Condition=Get OverrideStatusIndicatorState() -> (OverrideStatusIndicatorState:bool))
   Set Count(execute=@n10.else, Count=@n9.Count) -> (Output_Get=0)
   > n11 = Branch(Condition=Get OverrideStatusIndicatorVisibility())
   n12 = SetVisibility(execute=@n11.else, InVisibility=Select(Option 0=Collapsed, Option 1=HitTestInvisible, Index=@n9.bEnabled))
   Event PreConstruct()
   > InitializeStatusIndicator()
   > Branch(Condition=Select!(
     NewEnumerator0:bool=true
     NewEnumerator2:bool=false
     NewEnumerator3:bool=false
     Index:E_UI_StatusIndicator_Configuration Enum=Get Configuration() -> (Configuration=NewEnumerator0)
    ) -> (ReturnValue:bool))
   > SetIcon!(execute, self, IconAsset:Object Reference=Get Icon() -> (Icon:Object Reference)) -> (then)
   Event OnStopCallout() {is=title K2Node_Event}
   Event OnStartCallout!() -> (OutputDelegate, then, TutorialObjectiveName:Name=None, TutorialType:ETutorialType Enum=None) {is=title K2Node_Event}
  ]}
 Format Text {nodes [
   n1 = SetGridSlotPadding!(
    execute
    Widget:Widget Object Reference=Get Indicator_Text() -> (Indicator_Text:Common Text Object Reference)
    Padding_Left:float=10
    Padding_Top:float=0
    Padding_Right:float=10
    Padding_Bottom:float=0
   ) -> (then)
   n2 = Get Count() -> (Count=0)
   n3 = Get CountCeiling() -> (CountCeiling:int=0)
   "SetText (Text)"(
    execute=[@n1.then, @n6.then]
    self=Get Indicator_Text()
    InText=Select!(
     Option 0:Text="To Text (Integer)"(Value=@n2)
     Option 1:Text=Format Text!(Format="LOCTEXT(\"{0}+\")", 0:int=@n3) -> (Result)
     Index:bool="integer > integer"(A=@n2, B=@n3)
    ) -> (ReturnValue:Text)
   )
   > Set Justification!(execute, Justification:ETextJustify Enum=Center, self:Text Layout Widget Object Reference=Get Indicator_Text()) -> (then, Output_Get:ETextJustify Enum=Left)
   > Set MobileFontSizeMultiplier!(
    execute
    MobileFontSizeMultiplier:float=Get Mobile Font Size MultiplierNumeric() -> (Mobile Font Size MultiplierNumeric:float)
    self:Common Text Object Reference=Get Indicator_Text()
   ) -> (then, Output_Get:float)
   n4 = Get Font() -> (Font:Slate Font Info Structure)
   > n5 = Set members in Slate Font Info!(
    execute:Exec
    StructRef:"Slate Font Info Structure (by ref)"=Get Font()
    Size:float=Select Float with Boolean!(
     False:double=Break Slate Font Info!(SlateFontInfo:"Slate Font Info Structure (by ref)"=@n4) -> (Size:float=~, bMaterialIsStencil:bool=~) {is=title K2Node_BreakStruct}
     True:double=Get Mobile Font Size MultiplierNumeric()
     Boolean:bool=IsMobileGame()
    ) -> (Output:double)
   ) -> (then:Exec, StructOut:"Slate Font Info Structure (by ref)") {is=title K2Node_SetFieldsInStruct}
   > Set Font!(execute, Font:Slate Font Info Structure=@n5, self:Text Object Reference=Get Indicator_Text()) -> (then, Output_Get:Slate Font Info Structure)
   > SetColorAndOpacity!(
    execute
    self=Get Indicator_Text()
    InColorAndOpacity_SpecifiedColor:Linear Color Structure=Select!(
     NewEnumerator0:Linear Color Structure=(R=0,G=0.0331,B=0.0296,A=1)
     NewEnumerator2:Linear Color Structure=(R=0.0048,G=0.004,B=0.0116,A=1)
     NewEnumerator3:Linear Color Structure=(R=1,G=1,B=1,A=1)
     NewEnumerator4:Linear Color Structure=(R=0.1413,G=0.0212,B=0.0194,A=1)
     NewEnumerator5:Linear Color Structure=(R=0,G=0,B=0,A=1)
     Index:E_UI_StatusIndicator_Type Enum=Get Type() -> (Type=NewEnumerator0)
    ) -> (ReturnValue:Linear Color Structure)
    InColorAndOpacity_ColorUseRule:ESlateColorStylingMode Enum
   ) -> (then)
   n6 = Format Text() {is=title K2Node_FunctionEntry}
  ]}
 InitializeStatusIndicator {nodes [
   InitializeStatusIndicator() {is=title K2Node_FunctionEntry}
   > SetBackground(Material=Get BackgroundMaterial() -> (BackgroundMaterial:Material Interface Object Reference))
   > SetIndicatorConfiguration(InCount=Get Count() -> (Count=0), IndicatorConfig=Get Configuration() -> (Configuration=NewEnumerator0))
   > SetIndicatorType(IndicatorType=Get Type() -> (Type=NewEnumerator0))
  ]}
 PassBangTypeFromBP {nodes [
   n1 = PassBangTypeFromBP!() -> (then, EBangType:EFortBangType Enum=Invalid) {is=title K2Node_FunctionEntry}
   > SetBangType!(execute, self:Fort Bang Wrapper NUI Object Reference, NewBangType:EFortBangType Enum=@n1) -> (then)
  ]}
 Select Float with Boolean {nodes [
   n1 = Select Float with Boolean!() -> (then, False:double=~, True:double=~, Boolean:bool=~) {is=title K2Node_FunctionEntry}
   > Return Node!(
    execute
    Output:double=Select!(Option 0:float=@n1, Option 1:float=float * float(A=@n1, B=@n1.True) {is=title K2Node_PromotableOperator}, Index:bool=@n1.Boolean) -> (ReturnValue:float)
   )
  ]}
 Sequencer Events {links [], nodes []}
 SetBackground {nodes [
   n1 = SetBackground!() -> (then, Material:Material Interface Object Reference) {is=title K2Node_FunctionEntry}
   > n2 = Set BackgroundMaterial(BackgroundMaterial:Material Interface Object Reference=@n1) -> (Output_Get:Material Interface Object Reference)
   n3 = Get Image_BackgroundWithOutline() -> (Image_BackgroundWithOutline:Image Object Reference)
   > SetBrushFromMaterial(self=@n3, Material=@n2)
   > n4 = GetDynamicMaterial(self=@n3)
   > Set BackgroundDynamicMaterial(BackgroundDynamicMaterial:Material Instance Dynamic Object Reference=@n4) -> (Output_Get:Material Instance Dynamic Object Reference)
  ]}
 SetIcon {nodes [
   n1 = SetIcon!() -> (then, IconAsset:Object Reference) {is=title K2Node_FunctionEntry}
   > Set Icon(Icon:Object Reference=@n1) -> (Output_Get:Object Reference)
   > Get!(execute:Exec) -> (then:Exec, else:Exec, Indicator_Icon:Image Object Reference)
   n2 = Select!(
    NewEnumerator2:float=3
    NewEnumerator3:float=4
    Index:E_UI_StatusIndicator_Size Enum=Get IconSize() -> (IconSize:E_UI_StatusIndicator_Size Enum=NewEnumerator2)
   ) -> (ReturnValue)
   > SetGridSlotPadding!(
    execute
    Widget=Get Indicator_Icon() -> (Indicator_Icon:Image Object Reference)
    Padding:Margin Structure=Make Margin(Left=@n2, Top=@n2, Right=@n2, Bottom=@n2)
   ) -> (then)
   > SetColorAndOpacity!(
    execute
    self:Image Object Reference=Get Indicator_Icon()
    InColorAndOpacity:Linear Color Structure=Select!(
     NewEnumerator0:Linear Color Structure=(R=0,G=0.0331,B=0.0296,A=1)
     NewEnumerator2:Linear Color Structure=(R=0.0048,G=0.004,B=0.0116,A=1)
     NewEnumerator3:Linear Color Structure=(R=1,G=1,B=1,A=1)
     NewEnumerator4:Linear Color Structure=(R=0.1413,G=0.0212,B=0.0194,A=1)
     NewEnumerator5:Linear Color Structure=(R=0,G=0,B=0,A=1)
     Index:E_UI_StatusIndicator_Type Enum=Get Type() -> (Type=NewEnumerator0)
    ) -> (ReturnValue:Linear Color Structure)
   ) -> (then)
   n3 = IsMobileGame()
   n4 = Get MobileIconSizeMultiplierCompact() -> (MobileIconSizeMultiplierCompact:double)
   n5 = Get SizeIconCompact() -> (SizeIconCompact:double)
   n6 = Get MobileIconSizeMultipierDefault() -> (MobileIconSizeMultipierDefault:double)
   n7 = Get SizeIconDefault() -> (SizeIconDefault:double)
   n8 = Select!(
    NewEnumerator2:float=float - float(A=Select Float with Boolean(False=@n5, True=@n4, Boolean=@n3) {is=title K2Node_CallFunction}, B=6)
    NewEnumerator3:float=float - float(A=Select Float with Boolean(False=@n7, True=@n6, Boolean=@n3), B=-8)
    Index:E_UI_StatusIndicator_Size Enum=Get IconSize() -> (IconSize=NewEnumerator2)
   ) -> (ReturnValue:float)
   > SetBrush!(
    execute
    self=Get Indicator_Icon()
    InBrush_TintColor:Slate Color Structure=(SpecifiedColor=(R=1,G=1,B=1,A=1),ColorUseRule=UseColor_Specified)
    InBrush_DrawAs:ESlateBrushDrawType Enum=Image
    InBrush_Tiling:ESlateBrushTileType Enum=NoTile
    InBrush_Mirroring:ESlateBrushMirrorType Enum=NoMirror
    InBrush_ImageSize_X:double=@n8
    InBrush_ImageSize_Y:double=@n8
    InBrush_Margin:Margin Structure=(Left=0,Top=0,Right=0,Bottom=0)
    InBrush_ResourceObject:Object Reference=Get Icon()
    InBrush_OutlineSettings:Slate Brush Outline Settings Structure=(CornerRadii=(X=0,Y=0,Z=0,W=0),Color=(SpecifiedColor=(R=0,G=0,B=0,A=0),ColorUseRule=UseColor_Specified),Width=0,RoundingType=HalfHeightRadius,bUseBrushTransparency=False)
   ) -> (then)
  ]}
 SetIndicatorConfiguration {nodes [
   n1 = SetIndicatorConfiguration!() -> (then, InCount:int, IndicatorConfig:E_UI_StatusIndicator_Configuration Enum=NewEnumerator0) {is=title K2Node_FunctionEntry}
   > Set Configuration(Configuration:E_UI_StatusIndicator_Configuration Enum=@n1.IndicatorConfig) -> (Output_Get:E_UI_StatusIndicator_Configuration Enum=NewEnumerator0)
   > RemoveFromParent(self=[Get Indicator_Icon(), Get Indicator_Text()])
   > Set Indicator_Text(Indicator_Text:Common Text Object Reference) -> (Output_Get:Common Text Object Reference)
   > Set Indicator_Icon(Indicator_Icon:Image Object Reference) -> (Output_Get:Image Object Reference)
   > SetScalarParameterValue(
    self=Get BackgroundDynamicMaterial() -> (BackgroundDynamicMaterial:Material Instance Dynamic Object Reference)
    ParameterName=Outline Thickness
    Value=Select!(
     NewEnumerator0:float=0
     NewEnumerator2=2
     NewEnumerator3=0
     Index:E_UI_StatusIndicator_Configuration Enum=Get Configuration() -> (Configuration=NewEnumerator0)
    ) -> (ReturnValue)
   )
   > n2 = Sequence()
   > n3 = Check Current Configuration!(execute:Exec) -> (Image:Exec, Dot:Exec, Text:Exec) {is=title K2Node_MacroInstance}
   > n4 = Construct Object from Class(Class=&Image) -> (ReturnValue:Image Object Reference)
   > n5 = Set Indicator_Icon(Indicator_Icon=@n4)
   n6 = Get Root() -> (Root:Grid Panel Object Reference)
   > n7 = AddChildToGrid!(execute, self:Grid Panel Object Reference=@n6, Content:Widget Object Reference=@n5, InRow:int=0, InColumn:int=0) -> (then, ReturnValue:Grid Slot Object Reference)
   > Set L_ContentSlot(L_ContentSlot:Grid Slot Object Reference=@n7) -> (Output_Get:Grid Slot Object Reference)
   > n8 = SetIcon(IconAsset=Get Icon()) {is=title K2Node_CallFunction}
   SetLayer!(
    execute=[@n8.then, @n12.then]
    self:Grid Slot Object Reference=Get L_ContentSlot() -> (L_ContentSlot:Grid Slot Object Reference)
    InLayer:int=2
   ) -> (then)
   > SetVerticalAlignment(self:Grid Slot Object Reference=Get L_ContentSlot(), InVerticalAlignment=VAlign_Center)
   > SetHorizontalAlignment(self:Grid Slot Object Reference=Get L_ContentSlot(), InHorizontalAlignment=HAlign_Center)
   n9 = Construct Object from Class!(
    execute=@n3.Text
    Class=&CommonTextBlock
    self
    bIsScrollingEnabled:bool=false
    Style:Common Text Style Class Reference
    ScrollStyle:Common Text Scroll Style Class Reference
    ScrollOrientation:EOrientation Enum=Orient_Horizontal
   ) -> (then, ReturnValue:Common Text Object Reference)
   > n10 = Set Indicator_Text(Indicator_Text=@n9)
   > n11 = AddChildToGrid(self=@n6, Content=@n10, InRow=0, InColumn=0)
   > Set L_ContentSlot(L_ContentSlot=@n11)
   > n12 = Format Text() {is=title K2Node_CallFunction}
   UpdateSize!(execute=@n2.then_1, self, IconSize:E_UI_StatusIndicator_Size Enum=Get IconSize() -> (IconSize=NewEnumerator2)) -> (then)
  ]}
 SetIndicatorType {nodes [
   n1 = SetIndicatorType!() -> (then, IndicatorType:E_UI_StatusIndicator_Type Enum=NewEnumerator0) {is=title K2Node_FunctionEntry}
   > Set Type(Type:E_UI_StatusIndicator_Type Enum=@n1) -> (Output_Get:E_UI_StatusIndicator_Type Enum=NewEnumerator0)
   > SetVectorParameterValue(
    self=Get BackgroundDynamicMaterial()
    ParameterName=Color
    Value=Select!(
     NewEnumerator0:Linear Color Structure=(R=0.0395,G=1,B=0.2307,A=1)
     NewEnumerator2:Linear Color Structure=(R=0.9301,G=1,B=0.0103,A=1)
     NewEnumerator3:Linear Color Structure=(R=0.0048,G=0.004,B=0.0116,A=1)
     NewEnumerator4:Linear Color Structure=(R=1,G=0.147,B=0.1329,A=1)
     NewEnumerator5:Linear Color Structure=(R=1,G=0.8714,B=0.3813,A=1)
     Index:E_UI_StatusIndicator_Type Enum=Get Type() -> (Type=NewEnumerator0)
    ) -> (ReturnValue:Linear Color Structure)
   )
   > n2 = Check Current Configuration()
   > n3 = Get()
   > SetColorAndOpacity(
    self=@n3
    InColorAndOpacity=Select(
     NewEnumerator0=(R=0,G=0.0331,B=0.0296,A=1)
     NewEnumerator2=(R=0.0048,G=0.004,B=0.0116,A=1)
     NewEnumerator3=(R=1,G=1,B=1,A=1)
     NewEnumerator4=(R=0.1413,G=0.0212,B=0.0194,A=1)
     NewEnumerator5=(R=0,G=0,B=0,A=1)
     Index=Get Type() -> (Type=NewEnumerator0)
    )
   )
   n4 = Get!(execute=@n2.Text) -> (then, else, Indicator_Text:Common Text Object Reference)
   > SetColorAndOpacity!(
    execute
    self:Text Object Reference=@n4
    InColorAndOpacity_SpecifiedColor:Linear Color Structure=Select(
     NewEnumerator0=(R=0,G=0.0331,B=0.0296,A=1)
     NewEnumerator2=(R=0.0048,G=0.004,B=0.0116,A=1)
     NewEnumerator3=(R=1,G=1,B=1,A=1)
     NewEnumerator4=(R=0.1413,G=0.0212,B=0.0194,A=1)
     NewEnumerator5=(R=0,G=0,B=0,A=1)
     Index=Get Type() -> (Type=NewEnumerator0)
    )
    InColorAndOpacity_ColorUseRule:ESlateColorStylingMode Enum
   ) -> (then)
  ]}
 TriggerIntro {nodes [
   TriggerIntro() {is=title K2Node_FunctionEntry}
   > PlayAnimation(InAnimation=Get Anim_Intro() -> (Anim_Intro:Widget Animation Object Reference))
  ]}
 TriggerReset {nodes [
   TriggerReset() {is=title K2Node_FunctionEntry}
   > PlayAnimation(InAnimation=Get Anim_Intro(), PlayMode=Reverse)
  ]}
 UpdateSize {nodes [
   n1 = UpdateSize!() -> (then, IconSize:E_UI_StatusIndicator_Size Enum=NewEnumerator2) {is=title K2Node_FunctionEntry}
   > Set IconSize(IconSize:E_UI_StatusIndicator_Size Enum=@n1) -> (Output_Get:E_UI_StatusIndicator_Size Enum=NewEnumerator2)
   > Set L_IsMobile(L_IsMobile:bool=IsMobileGame()) -> (Output_Get:bool)
   > n2 = Check Current Configuration()
   n3 = Get L_IsMobile() -> (L_IsMobile:bool)
   n4 = Select!(
    NewEnumerator2:float=Select Float with Boolean(False=Get SizeIconCompact(), True=Get MobileIconSizeMultiplierCompact(), Boolean=@n3)
    NewEnumerator3:float=Select Float with Boolean(False=Get SizeIconDefault(), True=Get MobileIconSizeMultipierDefault(), Boolean=@n3)
    Index:E_UI_StatusIndicator_Size Enum=Get IconSize() -> (IconSize=NewEnumerator2)
   ) -> (ReturnValue:float)
   > n5 = Set!(execute, L_DesiretSize_X:double=@n4, L_DesiretSize_Y:double=@n4) -> (then, Output_Get:Vector 2D Structure)
   > SetDesiredSizeOverride!(execute, self:Image Object Reference=Get Image_BackgroundWithOutline(), DesiredSize:Vector 2D Structure=@n5) -> (then)
   > n6 = Get!(execute) -> (then, else, Indicator_Icon:Image Object Reference)
   n7 = Get L_IsMobile()
   n8 = Get MobileIconSizeMultiplierCompact()
   n9 = Get SizeIconCompact()
   n10 = Get MobileIconSizeMultipierDefault()
   n11 = Get SizeIconDefault()
   n12 = Select(
    NewEnumerator2=float - float(A=Select Float with Boolean(False=@n9, True=@n8, Boolean=@n7), B=6)
    NewEnumerator3=float - float(A=Select Float with Boolean(False=@n11, True=@n10, Boolean=@n7), B=8)
    Index=Get IconSize() -> (IconSize=NewEnumerator2)
   )
   > SetDesiredSizeOverride!(execute, self=@n6, DesiredSize_X:double=@n12, DesiredSize_Y:double=@n12) -> (then)
   n13 = Get Count() -> (Count=0)
   n14 = Branch(execute=@n2.Text, Condition="integer < integer"(A=@n13, B=10))
   n15 = Get L_IsMobile()
   > n16 = Set(
    L_DesiretSize_X=Select Float with Boolean(False=36, True=Get Mobile Font Size MultiplierNumeric(), Boolean=@n15)
    L_DesiretSize_Y=Select Float with Boolean(False=Get SizeNumber() -> (SizeNumber:double), True=Get Mobile Font Size MultiplierNumeric(), Boolean=@n15)
   )
   > SetDesiredSizeOverride!(execute, self=Get Image_BackgroundWithOutline(), DesiredSize:Vector 2D Structure=@n16) -> (then)
   n17 = Branch(execute=@n14.else, Condition="integer < integer"(A=@n13, B=100))
   n18 = Get L_IsMobile()
   > n19 = Set(
    L_DesiretSize_X=Select Float with Boolean(False=44, True=Get Mobile Font Size MultiplierNumeric(), Boolean=@n18)
    L_DesiretSize_Y=Select Float with Boolean(False=Get SizeNumber(), True=Get Mobile Font Size MultiplierNumeric(), Boolean=@n18)
   )
   > SetDesiredSizeOverride(self=Get Image_BackgroundWithOutline(), DesiredSize=@n19)
   n20 = Get Mobile Font Size MultiplierNumeric()
   n21 = Get L_IsMobile()
   n22 = Set(
    execute=@n17.else
    L_DesiretSize_X=Select Float with Boolean(False=58, True=@n20, Boolean=@n21)
    L_DesiretSize_Y=Select Float with Boolean(False=Get SizeNumber(), True=@n20, Boolean=@n21)
   )
   > SetDesiredSizeOverride(self=Get Image_BackgroundWithOutline(), DesiredSize=@n22)
   n23 = Get SizeDot() -> (SizeDot:float)
   n24 = Select!(
    Option 0:double=@n23
    Option 1:double=float * float(A=@n23, B=Get MobileIconSizeMultiplierDot() -> (MobileIconSizeMultiplierDot:double))
    Index:bool=Get L_IsMobile()
   ) -> (ReturnValue:double)
   n25 = Set(execute=@n2.Dot, L_DesiretSize_X=@n24, L_DesiretSize_Y=@n24)
   > SetDesiredSizeOverride(self=Get Image_BackgroundWithOutline(), DesiredSize=@n25)
  ]}
]
variables [
 BackgroundDynamicMaterial {type=MaterialInstanceDynamic}
 BackgroundMaterial {type=MaterialInterface, default=/FortUILibrary/Systems/StatusIndicator/MaterialInstances/MI_UIKit_StatusIndicator_Background}
 Configuration {type=E_UI_StatusIndicator_Configuration, default=Icon}
 Count {type=integer}
 CountCeiling {type=integer, default=99}
 Font {
  type=SlateFontInfo
  default {
   bForceMonospaced=false
   bMaterialIsStencil=false
   fontMaterial=null
   fontObject=&HeadingNow
   letterSpacing=0
   monospacedWidth=1
   size=15.75
   skewAmount=0
   typefaceFontName=76 Bold
   outlineSettings {
    bApplyOutlineToDropShadows=false
    bMiteredCorners=false
    bSeparateFillAlpha=false
    outlineMaterial=null
    outlineSize=0
    outlineColor=<rgba 0 0 0 1>
   }
  }
 }
 Icon {type=Object, default=/FortUILibrary/Textures/MaterialInstances/MI_UIKit_Icon_ExclamationMark}
 IconSize {type=E_UI_StatusIndicator_Size, default=Default}
 Indicator_Icon {type=Image, default=null}
 Indicator_Text {type=CommonTextBlock, default=null}
 Mobile Font Size MultiplierNumeric {type=number, default=1}
 MobileIconSizeMultipierDefault {type=number, default=1}
 MobileIconSizeMultiplierCompact {type=number, default=1}
 MobileIconSizeMultiplierDot {type=number, default=1}
 OverrideStatusIndicatorState {type=boolean}
 OverrideStatusIndicatorVisibility {type=boolean}
 SizeDot {type=number, default=14}
 SizeIconCompact {type=number, default=24}
 SizeIconDefault {type=number, default=32}
 SizeNumber {type=number, default=32}
 Type {type=E_UI_StatusIndicator_Type, default=Positive}
]
widgets {
 info {inheritedWidgetCount=1, parentClass=/Script/FortniteUI.FortBangWrapper_NUI, rootWidgetClass=&GridPanel, widgetCount=3}
 widgets [
  Root {widgetClassPath=&GridPanel, bIsVariable=true, widget=Root, properties {columnFill [1], rowFill [1]}}
  Image_BackgroundWithOutline {
   widgetClassPath=&Image
   bIsVariable=true
   parent=Root
   widget=Image_BackgroundWithOutline
   properties {
    visibility=SelfHitTestInvisible
    brush {
     bIsDynamicallyLoaded=false
     drawAs=Image
     imageType=FullColor
     mirroring=NoMirror
     resourceName=None
     resourceObject=null
     tiling=NoTile
     imageSize=<4 4>
     margin=<ltrb 0 0 0 0>
     outlineSettings {
      bUseBrushTransparency=false
      roundingType=HalfHeightRadius
      width=0
      color {colorUseRule=UseColor_Specified, specifiedColor=<rgba 0 0 0 0>}
      cornerRadii=<0 0 0 0>
     }
     tintColor {colorUseRule=UseColor_Specified, specifiedColor=<rgba 1 1 1 1>}
     uVRegion {bIsValid=false, max=<0 0>, min=<0 0>}
    }
   }
   slot {class=GridSlot, properties {horizontalAlignment=HAlign_Center, layer=1, verticalAlignment=VAlign_Center}}
  }
  ContentSlot {widgetClassPath=&NamedSlot, bInherited=true, widget=null}
 ]
}
